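# Looping, non-spatialized indoor daytime ambience cue for the metal room at TheRaft_FloatMax POI.
# Two mixes blend dock-interior and quieter exterior-platform loops with a shared industrial and large-room tone bed.
# A quality node picks the quad mix for its first two levels, stereo for the last two.
# Two nested looping envelopers add slow drift: pitch over 30 seconds, volume over 20 seconds.
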SoundCue /Game/Sounds/Helios/POI/TheRaft_FloatMax/Ambience_TheRaft_FloatMax_Metal_Room_Day_Cue
root=Enveloper_3
nodes [
 {
  id=Enveloper_0
  children [QualityLevel_1]
  properties {
   bLoop=true
   bLoopIndefinitely=true
   loopEnd=30
   pitchCurve {editorCurveData {keys [{interpMode=RCIM_Cubic, time=0}, <at 15.9468 0.9181 RCIM_Cubic>, {interpMode=RCIM_Cubic, time=30}]}}
   volumeCurve {editorCurveData {keys [<at 0 1>]}}
  }
 }
 {
  id=Enveloper_3
  children [Enveloper_0]
  properties {
   bLoop=true
   bLoopIndefinitely=true
   loopEnd=20
   pitchCurve {editorCurveData {keys [{time=0}]}}
   volumeCurve {editorCurveData {keys [
      {arriveTangent=-0.01, interpMode=RCIM_Cubic, leaveTangent=-0.01, time=0, value=1}
      <at 10 0.9 RCIM_Cubic>
      {arriveTangent=0.01, interpMode=RCIM_Cubic, leaveTangent=0.01, time=20, value=1}
     ]}}
  }
 }
 {id=Mixer_0, children [WavePlayer_3,WavePlayer_1], properties {inputVolume [0.75,0.75]}}
 {id=Mixer_1, children [WavePlayer_0,Mixer_0,Modulator_0], properties {inputVolume [0.2,0.75,0.1]}}
 {id=Mixer_2, children [WavePlayer_2,Mixer_0,Modulator_1], properties {inputVolume [0.2,0.75,0.1]}}
 {id=Modulator_0, children [WavePlayer_4], properties {volumeMax=0.9, volumeMin=0.9}}
 {id=Modulator_1, children [WavePlayer_5], properties {volumeMax=0.9, volumeMin=0.9}}
 {id=QualityLevel_1, children [Mixer_1,Mixer_1,Mixer_2,Mixer_2]}
 {
  id=WavePlayer_0
  properties {bLooping=true, soundWaveAssetPtr=/Game/Sounds/Artemis/POI/Hollow_Base_Takover/Ambience_Hollow_Base_Takeover_Dock_Interior_Day_Loop_Quad}
 }
 {
  id=WavePlayer_1
  properties {bLooping=true, soundWaveAssetPtr=/Game/Sounds/Fort_Starlight/S4/SoundWaves/AMBI_Starlight_S4_HauntedHotel_LargeRoom_Int_Night_Loop}
 }
 {
  id=WavePlayer_2
  properties {bLooping=true, soundWaveAssetPtr=/Game/Sounds/Artemis/POI/Hollow_Base_Takover/Ambience_Hollow_Base_Takeover_Dock_Interior_Day_Loop_Stereo}
 }
 {id=WavePlayer_3, properties {bLooping=true, soundWaveAssetPtr=/Game/Sounds/Trans_AmbientLoops/RoomTones/IndustrialSmall002_Loop_Stereo}}
 {
  id=WavePlayer_4
  properties {bLooping=true, soundWaveAssetPtr=/Game/Sounds/Artemis/POI/Hollow_Base_Takover/Ambience_Hollow_Base_Takeover_Platfrom_Ext_Day_Loop_Quad}
 }
 {
  id=WavePlayer_5
  properties {bLooping=true, soundWaveAssetPtr=/Game/Sounds/Artemis/POI/Hollow_Base_Takover/Ambience_Hollow_Base_Takeover_Platfrom_Ext_Day_Loop_Stereo}
 }
]
properties {
 bOverrideConcurrency=true
 maxDistance=2097152
 soundClassObject=&Fort_Ambient_Indoors_Athena
 volumeMultiplier=0.2
 attenuationOverrides {bAttenuate=false, bEnableReverbSend=false, bSpatialize=false}
 concurrencyOverrides {resolutionRule=PreventNew}
}
