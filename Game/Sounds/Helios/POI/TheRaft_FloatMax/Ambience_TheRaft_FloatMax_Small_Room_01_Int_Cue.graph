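# Looping indoor ambience Sound Cue for a small room at The Raft FloatMax POI, in the Athena indoor ambient sound class.
# A Quality Level node routes the first two levels to a quad loop and the last two to a stereo loop.
# Each loop passes through a Modulator that fixes pitch below normal.
# It plays quietly without attenuation, spatialization or reverb send, and its concurrency override prevents new instances over the limit.
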SoundCue /Game/Sounds/Helios/POI/TheRaft_FloatMax/Ambience_TheRaft_FloatMax_Small_Room_01_Int_Cue
root=QualityLevel_0
nodes [
 {id=Modulator_0, children [WavePlayer_0], properties {pitchMax=0.8, pitchMin=0.8}}
 {id=Modulator_1, children [WavePlayer_1], properties {pitchMax=0.8, pitchMin=0.8}}
 {id=QualityLevel_0, children [Modulator_0,Modulator_0,Modulator_1,Modulator_1]}
 {
  id=WavePlayer_0
  properties {bLooping=true, soundWaveAssetPtr=&/Ambience_TheRaft_Small_Room_01_Int_Quad}
 }
 {
  id=WavePlayer_1
  properties {bLooping=true, soundWaveAssetPtr=&/Ambience_TheRaft_Small_Room_01_Int_Stereo}
 }
]
properties {
 bOverrideConcurrency=true
 maxDistance=2097152
 soundClassObject=&Fort_Ambient_Indoors_Athena
 volumeMultiplier=0.226
 attenuationOverrides {bAttenuate=false, bEnableReverbSend=false, bSpatialize=false}
 concurrencyOverrides {resolutionRule=PreventNew}
}
