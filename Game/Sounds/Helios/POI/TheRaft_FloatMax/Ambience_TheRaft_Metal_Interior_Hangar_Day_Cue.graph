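# Sound cue for the non-spatialized, indoor daytime ambience bed of the metal hangar interior at The Raft POI.
# A quality-level node picks a quad or stereo version of a looping industrial wave reused from IcebergIsles ResearchStation.
# Two nested, endlessly looping envelopers of different periods add slow variation: the inner shapes pitch, the outer gently dips volume.
# The cue plays quietly without attenuation or reverb send, with concurrency preventing new instances at the limit.
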SoundCue /Game/Sounds/Helios/POI/TheRaft_FloatMax/Ambience_TheRaft_Metal_Interior_Hangar_Day_Cue
root=Enveloper_3
nodes [
 {
  id=Enveloper_0
  children [QualityLevel_1]
  properties {
   bLoop=true
   bLoopIndefinitely=true
   loopEnd=30
   pitchCurve {editorCurveData {keys [{interpMode=RCIM_Cubic, time=0}, <at 15.9468 0.9181 RCIM_Cubic>, {interpMode=RCIM_Cubic, time=30}]}}
   volumeCurve {editorCurveData {keys [<at 0 1>]}}
  }
 }
 {
  id=Enveloper_3
  children [Enveloper_0]
  properties {
   bLoop=true
   bLoopIndefinitely=true
   loopEnd=20
   pitchCurve {editorCurveData {keys [{time=0}]}}
   volumeCurve {editorCurveData {keys [
      {arriveTangent=-0.01, interpMode=RCIM_Cubic, leaveTangent=-0.01, time=0, value=1}
      <at 10 0.9 RCIM_Cubic>
      {arriveTangent=0.01, interpMode=RCIM_Cubic, leaveTangent=0.01, time=20, value=1}
     ]}}
  }
 }
 {id=QualityLevel_1, children [WavePlayer_0,WavePlayer_0,WavePlayer_1,WavePlayer_1]}
 {
  id=WavePlayer_0
  properties {bLooping=true, soundWaveAssetPtr=/Game/Sounds/Asteria/POI/IcebergIsles_ResearchStation/Ambience_IndustrialLarge001_Loop_Quad}
 }
 {
  id=WavePlayer_1
  properties {bLooping=true, soundWaveAssetPtr=/Game/Sounds/Asteria/POI/IcebergIsles_ResearchStation/Ambience_IndustrialLarge001_Loop_Stereo}
 }
]
properties {
 bOverrideConcurrency=true
 maxDistance=2097152
 soundClassObject=&Fort_Ambient_Indoors_Athena
 volumeMultiplier=0.125
 attenuationOverrides {bAttenuate=false, bEnableReverbSend=false, bSpatialize=false}
 concurrencyOverrides {resolutionRule=PreventNew}
}
