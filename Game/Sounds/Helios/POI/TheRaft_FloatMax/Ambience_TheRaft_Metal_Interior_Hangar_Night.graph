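# Looping, non-spatialized indoor night ambience cue for the metal hangar interior at The Raft POI, reusing Iceberg Isles research station loops.
# Crossfades a pitched-down industrial loop against a mine loop using opposed 30-second volume envelopes, with the industrial layer mixed at half volume.
# A modulator randomizes each mix's pitch and volume.
# A quality-level node picks quad waves for the first two levels and stereo for the last two.
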SoundCue /Game/Sounds/Helios/POI/TheRaft_FloatMax/Ambience_TheRaft_Metal_Interior_Hangar_Night
root=QualityLevel_1
nodes [
 {
  id=Enveloper_1
  children [WavePlayer_4]
  properties {
   bLoop=true
   bLoopIndefinitely=true
   loopEnd=30
   pitchCurve {editorCurveData {keys [{time=0}]}}
   volumeCurve {editorCurveData {keys [<at 0 1 RCIM_Cubic>,<at 15 0.3 RCIM_Cubic>,<at 30 1 RCIM_Cubic>]}}
  }
 }
 {
  id=Enveloper_2
  children [Modulator_2]
  properties {
   bLoop=true
   bLoopIndefinitely=true
   loopEnd=30
   pitchCurve {editorCurveData {keys [{time=0}]}}
   volumeCurve {editorCurveData {keys [<at 0 0.3 RCIM_Cubic>,<at 15 1 RCIM_Cubic>,<at 30 0.3 RCIM_Cubic>]}}
  }
 }
 {
  id=Enveloper_6
  children [WavePlayer_2]
  properties {
   bLoop=true
   bLoopIndefinitely=true
   loopEnd=30
   pitchCurve {editorCurveData {keys [{time=0}]}}
   volumeCurve {editorCurveData {keys [<at 0 1 RCIM_Cubic>,<at 15 0.3 RCIM_Cubic>,<at 30 1 RCIM_Cubic>]}}
  }
 }
 {
  id=Enveloper_7
  children [Modulator_0]
  properties {
   bLoop=true
   bLoopIndefinitely=true
   loopEnd=30
   pitchCurve {editorCurveData {keys [{time=0}]}}
   volumeCurve {editorCurveData {keys [<at 0 0.3 RCIM_Cubic>,<at 15 1 RCIM_Cubic>,<at 30 0.3 RCIM_Cubic>]}}
  }
 }
 {id=Mixer_0, children [Enveloper_2,Enveloper_1], properties {inputVolume [0.5,1]}}
 {id=Mixer_1, children [Enveloper_7,Enveloper_6], properties {inputVolume [0.5,1]}}
 {id=Modulator_0, children [WavePlayer_0], properties {pitchMax=0.8, pitchMin=0.5}}
 {id=Modulator_1, children [Mixer_0], properties {pitchMax=1.2, pitchMin=0.8, volumeMax=1.1, volumeMin=0.9}}
 {id=Modulator_12, children [Mixer_1], properties {pitchMax=1.2, pitchMin=0.8, volumeMax=1.1, volumeMin=0.9}}
 {id=Modulator_2, children [WavePlayer_3], properties {pitchMax=0.8, pitchMin=0.5}}
 {id=QualityLevel_1, children [Modulator_12,Modulator_12,Modulator_1,Modulator_1]}
 {
  id=WavePlayer_0
  properties {bLooping=true, soundWaveAssetPtr=/Game/Sounds/Asteria/POI/IcebergIsles_ResearchStation/Ambience_IndustrialLarge001_Loop_Quad}
 }
 {
  id=WavePlayer_2
  properties {bLooping=true, soundWaveAssetPtr=/Game/Sounds/Asteria/POI/IcebergIsles_ResearchStation/Ambience_MineA012_Loop_Quad}
 }
 {
  id=WavePlayer_3
  properties {bLooping=true, soundWaveAssetPtr=/Game/Sounds/Asteria/POI/IcebergIsles_ResearchStation/Ambience_IndustrialLarge001_Loop_Stereo}
 }
 {
  id=WavePlayer_4
  properties {bLooping=true, soundWaveAssetPtr=/Game/Sounds/Asteria/POI/IcebergIsles_ResearchStation/Ambience_MineA012_Loop_Stereo}
 }
]
properties {
 bOverrideConcurrency=true
 maxDistance=2097152
 soundClassObject=&Fort_Ambient_Indoors_Athena
 volumeMultiplier=0.125
 attenuationOverrides {bAttenuate=false, bEnableReverbSend=false, bSpatialize=false}
 concurrencyOverrides {resolutionRule=PreventNew}
}
